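# EQS query that picks a random human player from the querying pawn's squad, for the Supply NPC specialist to supply.
# The squad members generator excludes the querier, AI pawns, NPCs and player bots, and a random test scores the candidates.
# A 3D distance test only filters by range from the querier, despite its "Prefer closer items" comment.
# Its maximum range is bound to a float curve over game difficulty.
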EnvQuery /NPCSpecialists/NPCs/Supply/EQS/EQS_FindRandomSquadMemberToSupply
options [{
  generator {
   name=FortQueryGenerator_SquadMembers_1
   class=FortQueryGenerator_SquadMembers
   properties {
    bAutoSortTests=true
    bCanRunAsync=false
    envQueryResultNormalizationOption=Default
    includeSelf=false
    lookingForAiPawns=false
    lookingForHumanPlayers=true
    lookingForNpcs=false
    lookingForPlayerBots=false
    optionName=""
   }
  }
  tests [
   FortQueryTest_Random_0 {class=FortQueryTest_Random, properties {floatValueMax=0, scoringFactor=1}}
   EnvQueryTest_Distance_0 {
    class=EnvQueryTest_Distance
    properties {
     clampMaxType=SpecifiedValue
     distanceTo=&EnvQueryContext_Querier
     filterType=Range
     floatValueMin=0
     scoringEquation=InverseLinear
     scoringFactor=1
     testComment=Prefer closer items
     testMode=Distance3D
     testPurpose=Filter
     floatValueMax {
      dataBinding=&self:EnvQueryTest_Distance_0.FortAIDataProvider_FloatCurveOverGameDifficulty_3
      dataField=FloatValue
      defaultValue=2500
     }
     scoreClampMax {
      dataBinding=&self:EnvQueryTest_Distance_0.FortAIDataProvider_FloatCurveOverGameDifficulty_2
      dataField=FloatValue
      defaultValue=2500
     }
     scoreClampMin {
      dataBinding=&self:EnvQueryTest_Distance_0.FortAIDataProvider_FloatCurveOverGameDifficulty_1
      dataField=FloatValue
      defaultValue=0
     }
    }
   }
  ]
 }]
